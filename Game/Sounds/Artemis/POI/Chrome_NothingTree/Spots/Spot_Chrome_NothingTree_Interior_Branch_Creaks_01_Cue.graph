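# Sound cue for branch creak spot ambience inside the Chrome Nothing Tree point of interest.
# A quality-level switch sends its first two slots to an equally weighted random pick of eight creak waves, and its last two to a pick of three (creaks 01, 02, 04).
# The three-wave branch is excluded from branch culling.
# Playback uses a wider one-shot ambient attenuation, the outdoor ambient sound class and boosted volume.
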SoundCue /Game/Sounds/Artemis/POI/Chrome_NothingTree/Spots/Spot_Chrome_NothingTree_Interior_Branch_Creaks_01_Cue
root=QualityLevel_0
nodes [
 {id=QualityLevel_0, children [Random_2,Random_2,Random_3,Random_3]}
 {
  id=Random_2
  children [WavePlayer_13, WavePlayer_14, WavePlayer_15, WavePlayer_16, WavePlayer_17, WavePlayer_18, WavePlayer_19, WavePlayer_20]
  properties {weights [1,1,1,1,1,1,1,1]}
 }
 {id=Random_3, children [WavePlayer_21, WavePlayer_22, WavePlayer_23], properties {bShouldExcludeFromBranchCulling=true, weights [1,1,1]}}
 {id=WavePlayer_13, properties {soundWaveAssetPtr=&/Spot_Chrome_NothingTree_Branch_Creak_08}}
 {id=WavePlayer_14, properties {soundWaveAssetPtr=&/Spot_Chrome_NothingTree_Branch_Creak_01}}
 {id=WavePlayer_15, properties {soundWaveAssetPtr=&/Spot_Chrome_NothingTree_Branch_Creak_02}}
 {id=WavePlayer_16, properties {soundWaveAssetPtr=&/Spot_Chrome_NothingTree_Branch_Creak_03}}
 {id=WavePlayer_17, properties {soundWaveAssetPtr=&/Spot_Chrome_NothingTree_Branch_Creak_04}}
 {id=WavePlayer_18, properties {soundWaveAssetPtr=&/Spot_Chrome_NothingTree_Branch_Creak_05}}
 {id=WavePlayer_19, properties {soundWaveAssetPtr=&/Spot_Chrome_NothingTree_Branch_Creak_06}}
 {id=WavePlayer_20, properties {soundWaveAssetPtr=&/Spot_Chrome_NothingTree_Branch_Creak_07}}
 {id=WavePlayer_21, properties {soundWaveAssetPtr=&Spot_Chrome_NothingTree_Branch_Creak_02}}
 {id=WavePlayer_22, properties {soundWaveAssetPtr=&Spot_Chrome_NothingTree_Branch_Creak_04}}
 {id=WavePlayer_23, properties {soundWaveAssetPtr=&Spot_Chrome_NothingTree_Branch_Creak_01}}
]
properties {
 attenuationSettings=/Game/Sounds/Attenuation/Fort_TileAmbient_OneShot_Wider_Att
 maxDistance=8000
 soundClassObject=&Fort_Ambient_Outdoors_Athena
 volumeMultiplier=1.504
 preEffectBusSends [{}]
}
