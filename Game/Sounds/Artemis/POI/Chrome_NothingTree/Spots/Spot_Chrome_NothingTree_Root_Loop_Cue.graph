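# Sound cue for outdoor ambient root spot sounds at the Chrome NothingTree POI, reusing Reality Forest root samples.
# Two parallel Random nodes draw from the same eleven waves, each through pitch and volume modulation, then mix into a layered variation.
# A quality switch feeds that mix to its first two slots, leaving the others empty.
# Overrides add occluded attenuation and cap concurrency; no loop node exists despite the name.
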SoundCue /Game/Sounds/Artemis/POI/Chrome_NothingTree/Spots/Spot_Chrome_NothingTree_Root_Loop_Cue
root=QualityLevel_0
nodes [
 {id=Mixer_1, children [Modulator_0,Modulator_3], properties {inputVolume [1,1]}}
 {id=Modulator_0, children [Random_0], properties {pitchMin=0.85, volumeMax=1.25, volumeMin=0.7}}
 {id=Modulator_3, children [Random_1], properties {pitchMin=0.85, volumeMax=1.25, volumeMin=0.7}}
 {id=QualityLevel_0, children [Mixer_1,Mixer_1,null,null]}
 {
  id=Random_0
  children [
   WavePlayer_0
   WavePlayer_1
   WavePlayer_2
   WavePlayer_3
   WavePlayer_4
   WavePlayer_5
   WavePlayer_6
   WavePlayer_7
   WavePlayer_8
   WavePlayer_9
   WavePlayer_22
  ]
  properties {weights [1,1,1,1,1,1,1,1,1,1,1]}
 }
 {
  id=Random_1
  children [
   WavePlayer_23
   WavePlayer_24
   WavePlayer_25
   WavePlayer_26
   WavePlayer_27
   WavePlayer_28
   WavePlayer_29
   WavePlayer_30
   WavePlayer_31
   WavePlayer_32
   WavePlayer_33
  ]
  properties {weights [1,1,1,1,1,1,1,1,1,1,1]}
 }
 {id=WavePlayer_0, properties {soundWaveAssetPtr=/Game/Sounds/Artemis/Biomes/Reality_Forest/Spots/AMB_Reality_Forest_Root_SS_11}}
 {id=WavePlayer_1, properties {soundWaveAssetPtr=/Game/Sounds/Artemis/Biomes/Reality_Forest/Spots/AMB_Reality_Forest_Root_SS_01}}
 {id=WavePlayer_2, properties {soundWaveAssetPtr=/Game/Sounds/Artemis/Biomes/Reality_Forest/Spots/AMB_Reality_Forest_Root_SS_02}}
 {id=WavePlayer_22, properties {soundWaveAssetPtr=/Game/Sounds/Artemis/Biomes/Reality_Forest/Spots/AMB_Reality_Forest_Root_SS_10}}
 {id=WavePlayer_23, properties {soundWaveAssetPtr=&AMB_Reality_Forest_Root_SS_11}}
 {id=WavePlayer_24, properties {soundWaveAssetPtr=&AMB_Reality_Forest_Root_SS_01}}
 {id=WavePlayer_25, properties {soundWaveAssetPtr=&AMB_Reality_Forest_Root_SS_02}}
 {id=WavePlayer_26, properties {soundWaveAssetPtr=/Game/Sounds/Artemis/Biomes/Reality_Forest/Spots/AMB_Reality_Forest_Root_SS_03}}
 {id=WavePlayer_27, properties {soundWaveAssetPtr=/Game/Sounds/Artemis/Biomes/Reality_Forest/Spots/AMB_Reality_Forest_Root_SS_04}}
 {id=WavePlayer_28, properties {soundWaveAssetPtr=/Game/Sounds/Artemis/Biomes/Reality_Forest/Spots/AMB_Reality_Forest_Root_SS_05}}
 {id=WavePlayer_29, properties {soundWaveAssetPtr=/Game/Sounds/Artemis/Biomes/Reality_Forest/Spots/AMB_Reality_Forest_Root_SS_06}}
 {id=WavePlayer_3, properties {soundWaveAssetPtr=&AMB_Reality_Forest_Root_SS_03}}
 {id=WavePlayer_30, properties {soundWaveAssetPtr=/Game/Sounds/Artemis/Biomes/Reality_Forest/Spots/AMB_Reality_Forest_Root_SS_07}}
 {id=WavePlayer_31, properties {soundWaveAssetPtr=/Game/Sounds/Artemis/Biomes/Reality_Forest/Spots/AMB_Reality_Forest_Root_SS_08}}
 {id=WavePlayer_32, properties {soundWaveAssetPtr=/Game/Sounds/Artemis/Biomes/Reality_Forest/Spots/AMB_Reality_Forest_Root_SS_09}}
 {id=WavePlayer_33, properties {soundWaveAssetPtr=&AMB_Reality_Forest_Root_SS_10}}
 {id=WavePlayer_4, properties {soundWaveAssetPtr=&AMB_Reality_Forest_Root_SS_04}}
 {id=WavePlayer_5, properties {soundWaveAssetPtr=&AMB_Reality_Forest_Root_SS_05}}
 {id=WavePlayer_6, properties {soundWaveAssetPtr=&AMB_Reality_Forest_Root_SS_06}}
 {id=WavePlayer_7, properties {soundWaveAssetPtr=&AMB_Reality_Forest_Root_SS_07}}
 {id=WavePlayer_8, properties {soundWaveAssetPtr=&AMB_Reality_Forest_Root_SS_08}}
 {id=WavePlayer_9, properties {soundWaveAssetPtr=&AMB_Reality_Forest_Root_SS_09}}
]
properties {
 attenuationSettings=/Game/Sounds/Attenuation/Fort_TileAmbient_OneShot_Wider_Att
 bEnableBusSends=false
 bOverrideAttenuation=true
 bOverrideConcurrency=true
 maxDistance=1700
 pitchMultiplier=0.9
 soundClassObject=&Fort_Ambient_Outdoors_Athena
 volumeMultiplier=0.752
 attenuationOverrides {
  bEnableOcclusion=true
  falloffDistance=1500
  occlusionLowPassFilterFrequency=2000
  occlusionTraceChannel=ECC_GameTraceChannel11
  occlusionVolumeAttenuation=0.5
  attenuationShapeExtents {x=200}
 }
 concurrencyOverrides {maxCount=2}
 preEffectBusSends [{}]
}
