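# Sound cue for a desolate wind spot ambience one-shot at the Chrome NothingTree POI.
# A Random node picks among five equally weighted long cold-wind one-shots borrowed from the Apollo winter weather set.
# A Modulator then varies the volume widely, from faint to moderate, on each play.
# Output uses the outdoor ambient one-shot sound class and a wide tile-ambient one-shot attenuation.
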SoundCue /Game/Sounds/Artemis/POI/Chrome_NothingTree/Spots/Spot_NothingTree_Desolate_Wind_Cue
root=Modulator_1
nodes [
 {id=Modulator_1, children [Random_1], properties {volumeMax=0.7, volumeMin=0.1}}
 {id=Random_1, children [WavePlayer_0,WavePlayer_6,WavePlayer_7,WavePlayer_8,WavePlayer_9], properties {weights [1,1,1,1,1]}}
 {id=WavePlayer_0, properties {soundWaveAssetPtr=/Game/Sounds/Apollo/Weather/Winter/OneShots/Winter_Oneshot_ColdWind_Long_05}}
 {id=WavePlayer_6, properties {soundWaveAssetPtr=/Game/Sounds/Apollo/Weather/Winter/OneShots/Winter_Oneshot_ColdWind_Long_04}}
 {id=WavePlayer_7, properties {soundWaveAssetPtr=/Game/Sounds/Apollo/Weather/Winter/OneShots/Winter_Oneshot_ColdWind_Long_03}}
 {id=WavePlayer_8, properties {soundWaveAssetPtr=/Game/Sounds/Apollo/Weather/Winter/OneShots/Winter_Oneshot_ColdWind_Long_02}}
 {id=WavePlayer_9, properties {soundWaveAssetPtr=/Game/Sounds/Apollo/Weather/Winter/OneShots/Winter_Oneshot_ColdWind_Long_01}}
]
properties {
 attenuationSettings=/Game/Sounds/Attenuation/Fort_TileAmbient_OneShot_Wider_Att
 maxDistance=8000
 soundClassObject=&Fort_Ambient_OneShots_Outdoors
 volumeMultiplier=1.002
 preEffectBusSends [{}]
}
