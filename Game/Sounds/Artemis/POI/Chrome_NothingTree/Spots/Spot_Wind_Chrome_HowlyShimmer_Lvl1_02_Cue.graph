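# Sound cue for a one-shot howly, shimmering wind spot ambience at the Chrome Nothing Tree POI, drawing on Chrome biome wave assets.
# A Random node picks among ten equally weighted wind variations.
# Every quality level branch routes to that same Random node, and a Modulator at the root slightly varies volume.
# It plays in the outdoor ambient sound class with a wide tile-ambient one-shot attenuation and reduced overall volume.
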SoundCue /Game/Sounds/Artemis/POI/Chrome_NothingTree/Spots/Spot_Wind_Chrome_HowlyShimmer_Lvl1_02_Cue
root=Modulator_0
nodes [
 {id=Modulator_0, children [QualityLevel_0], properties {volumeMin=0.9}}
 {id=QualityLevel_0, children [Random_1,Random_1,Random_1,Random_1]}
 {
  id=Random_1
  children [
   WavePlayer_10
   WavePlayer_11
   WavePlayer_12
   WavePlayer_13
   WavePlayer_14
   WavePlayer_15
   WavePlayer_16
   WavePlayer_17
   WavePlayer_18
   WavePlayer_19
  ]
  properties {weights [1,1,1,1,1,1,1,1,1,1]}
 }
 {id=WavePlayer_10, properties {soundWaveAssetPtr=/Game/Sounds/Artemis/Biomes/Chrome/Spots/Spot_Wind_Chrome_HowlyShimmer_Lvl1_02_01}}
 {id=WavePlayer_11, properties {soundWaveAssetPtr=/Game/Sounds/Artemis/Biomes/Chrome/Spots/Spot_Wind_Chrome_HowlyShimmer_Lvl1_02_02}}
 {id=WavePlayer_12, properties {soundWaveAssetPtr=/Game/Sounds/Artemis/Biomes/Chrome/Spots/Spot_Wind_Chrome_HowlyShimmer_Lvl1_02_03}}
 {id=WavePlayer_13, properties {soundWaveAssetPtr=/Game/Sounds/Artemis/Biomes/Chrome/Spots/Spot_Wind_Chrome_HowlyShimmer_Lvl1_02_04}}
 {id=WavePlayer_14, properties {soundWaveAssetPtr=/Game/Sounds/Artemis/Biomes/Chrome/Spots/Spot_Wind_Chrome_HowlyShimmer_Lvl1_02_05}}
 {id=WavePlayer_15, properties {soundWaveAssetPtr=/Game/Sounds/Artemis/Biomes/Chrome/Spots/Spot_Wind_Chrome_HowlyShimmer_Lvl1_02_06}}
 {id=WavePlayer_16, properties {soundWaveAssetPtr=/Game/Sounds/Artemis/Biomes/Chrome/Spots/Spot_Wind_Chrome_HowlyShimmer_Lvl1_02_07}}
 {id=WavePlayer_17, properties {soundWaveAssetPtr=/Game/Sounds/Artemis/Biomes/Chrome/Spots/Spot_Wind_Chrome_HowlyShimmer_Lvl1_02_08}}
 {id=WavePlayer_18, properties {soundWaveAssetPtr=/Game/Sounds/Artemis/Biomes/Chrome/Spots/Spot_Wind_Chrome_HowlyShimmer_Lvl1_02_09}}
 {id=WavePlayer_19, properties {soundWaveAssetPtr=/Game/Sounds/Artemis/Biomes/Chrome/Spots/Spot_Wind_Chrome_HowlyShimmer_Lvl1_02_10}}
]
properties {
 attenuationSettings=/Game/Sounds/Attenuation/Fort_TileAmbient_OneShot_Wider_Att
 bEnableBusSends=false
 maxDistance=8000
 soundClassObject=&Fort_Ambient_Outdoors_Athena
 volumeMultiplier=0.501
 preEffectBusSends [{}]
}
